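# Physics asset for the left leg strap on the Shark Attack medium male player body cosmetic, defining bodies for dangling strap physics.
# A constraint chain runs from the pelvis capsule through a small base capsule and strap capsule to a box ring, with tight swing and twist limits on the last two joints.
# A large strapCol capsule, apparently a collision proxy, also attaches to the pelvis; both pelvis constraints leave swing and twist free.
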PhysicsAsset /BRCosmetics/Characters/Player/Male/Medium/Bodies/M_MED_Shark_Attack/Meshes/M_MED_Shark_Attack_LegStrap_Physics
bodies {
 dyn_base_legStrap_base_l [{shapeType=Capsule, length=1, radius=1}]
 dyn_legStrap_l [{shapeType=Capsule, length=4.4616, radius=1.395, center=<3.0913 0.4562 -0.0545>, rotation=<rot 0 -91.1904 -87.8942>}]
 dyn_legStrap_ring_l [{shapeType=Box, extentX=4, extentY=2.5, extentZ=6, center=<1.272 0 0>}]
 dyn_strapCol [{shapeType=Capsule, length=13.5, radius=8.255, center=<-9.446 0.9039 0>, rotation=<rot 90 1.7899 -0.3668>}]
 pelvis [{shapeType=Capsule, length=11.4912, radius=12.0588, center=<-3.6397 -0.406 0.3933>}]
}
constraints [
 {bone1Name=dyn_legStrap_l, bone2Name=dyn_base_legStrap_base_l, swing1LimitDegrees=30, swing2LimitDegrees=15, twistLimitDegrees=5}
 {bone1Name=dyn_legStrap_ring_l, bone2Name=dyn_legStrap_l, swing1LimitDegrees=20, swing2LimitDegrees=5, twistLimitDegrees=5}
 {
  bone1Name=dyn_strapCol
  bone2Name=pelvis
  swing1LimitDegrees=45
  swing1Motion=Free
  swing2LimitDegrees=45
  swing2Motion=Free
  twistLimitDegrees=45
  twistMotion=Free
 }
 {
  bone1Name=dyn_base_legStrap_base_l
  bone2Name=pelvis
  swing1LimitDegrees=45
  swing1Motion=Free
  swing2LimitDegrees=45
  swing2Motion=Free
  twistLimitDegrees=45
  twistMotion=Free
 }
]
