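# Physics asset for the Shark Attack medium male player body, set up for secondary motion on its dangling gear.
# Capsules on pelvis, spine_03 and both thighs form a coarse body, plus a strap collision capsule on the left thigh.
# Small base capsules link to these body parts with free angular constraints.
# Chest straps and rings, left shoulder piece, grenade and pouches hang from those bases with limited swing and twist.
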PhysicsAsset /BRCosmetics/Characters/Player/Male/Medium/Bodies/M_MED_Shark_Attack/Meshes/M_MED_Shark_Attack_Physics
bodies {
 dyn_chest_ring_l [{shapeType=Box, extentX=3, extentY=2, extentZ=4, center=<1.4167 0 0>}]
 dyn_chest_ring_r [{shapeType=Box, extentX=3, extentY=2, extentZ=4, center=<-1.4167 0 0>, rotation=<rot 0 0 180>}]
 dyn_chest_strap_a_base_l [{shapeType=Capsule, length=1, radius=1}]
 dyn_chest_strap_a_base_r [{shapeType=Capsule, length=0.505, radius=0.505, rotation=<rot 0 -90 0>}]
 dyn_chest_strap_a_l [{shapeType=Box, extentX=3.5, extentY=1.5, extentZ=3, center=<2.0995 0.0001 0>}]
 dyn_chest_strap_a_r [{shapeType=Box, extentX=3.5, extentY=1.5, extentZ=3, center=<-2.0995 -0.0001 0>, rotation=<rot 0 0 180>}]
 dyn_chest_strap_b_l [{shapeType=Box, extentX=4, extentY=2, extentZ=3, center=<2.0336 0 0>}]
 dyn_chest_strap_b_r [{shapeType=Box, extentX=4, extentY=2, extentZ=3, center=<-2.0336 0 0>, rotation=<rot 0 0 180>}]
 dyn_grenade_base_l [{shapeType=Capsule, length=1, radius=1}]
 dyn_grenade_l [{shapeType=Capsule, length=1, radius=3.5, center=<6.9951 0 0>, rotation=<rot 90 0 0>}]
 dyn_pouchA_base_r [{shapeType=Capsule, length=1, radius=1}]
 dyn_pouchA_r [{shapeType=Box, extentX=10, extentY=3.5, extentZ=4.5, center=<1.503 1.48 0.1757>}]
 dyn_pouchB_base_l [{shapeType=Capsule, length=1, radius=1}]
 dyn_pouchB_base_r [{shapeType=Capsule, length=1, radius=1}]
 dyn_pouchB_l [{shapeType=Box, extentX=10.5, extentY=5.5, extentZ=4, center=<2.6366 0.6922 2.4582>}]
 dyn_pouchB_r [{shapeType=Box, extentX=10, extentY=3, extentZ=5, center=<2.0383 1.6089 -0.2047>}]
 dyn_pouchC_base_l [{shapeType=Capsule, length=1, radius=1}]
 dyn_pouchC_base_r [{shapeType=Capsule, length=1, radius=1}]
 dyn_pouchC_l [{shapeType=Box, extentX=15.5, extentY=8, extentZ=5.5, center=<-0.7565 1.7493 1.6335>, rotation=<rot 0 0 12.7689>}]
 dyn_pouchC_r [{shapeType=Box, extentX=9.5, extentY=6, extentZ=4, center=<0.9513 -0.0001 1.6115>}]
 dyn_shldrA_base_l [{shapeType=Capsule, length=1, radius=1}]
 dyn_shldrA_l [{shapeType=Box, extentX=4.5, extentY=7.5, extentZ=2, center=<2.1194 0 0>}]
 dyn_shldrA_ring_l [{shapeType=Box, extentX=4, extentY=6.5, extentZ=2, center=<1.7009 -0.0001 0>}]
 dyn_strapCol [{shapeType=Capsule, length=13.5, radius=8.255, center=<-10.1724 0.8766 0.0001>, rotation=<rot 90 -1.7899 -3.9466>}]
 pelvis [{shapeType=Capsule, length=7.7276, radius=11.8039, center=<-1.7979 -0.027 0.7283>, rotation=<rot -3.1226 -0.709 -177.2301>}]
 spine_03 [{shapeType=Capsule, length=6, radius=14.75, rotation=<rot 90 0 0>}]
 thigh_l [{shapeType=Capsule, length=36.25, radius=8.25, center=<-18.7876 0.0004 0>, rotation=<rot 90 -4.4672 -3.0225>}]
 thigh_r [{shapeType=Capsule, length=26.5, radius=8.5, center=<21.7188 0.2423 0.2329>, rotation=<rot 88.0275 -6.491 168.7389>}]
}
constraints [
 {bone1Name=dyn_chest_strap_a_l, bone2Name=dyn_chest_strap_a_base_l, swing1LimitDegrees=10, swing2LimitDegrees=10, twistLimitDegrees=5}
 {bone1Name=dyn_chest_strap_b_l, bone2Name=dyn_chest_strap_a_l, swing1LimitDegrees=10, swing2LimitDegrees=10, twistLimitDegrees=5}
 {bone1Name=dyn_chest_ring_l, bone2Name=dyn_chest_strap_b_l, swing1LimitDegrees=15, swing2LimitDegrees=5, twistLimitDegrees=5}
 {bone1Name=dyn_chest_strap_a_r, bone2Name=dyn_chest_strap_a_base_r, swing1LimitDegrees=10, swing2LimitDegrees=10, twistLimitDegrees=5}
 {bone1Name=dyn_chest_strap_b_r, bone2Name=dyn_chest_strap_a_r, swing1LimitDegrees=10, swing2LimitDegrees=10, twistLimitDegrees=5}
 {bone1Name=dyn_chest_ring_r, bone2Name=dyn_chest_strap_b_r, swing1LimitDegrees=15, swing2LimitDegrees=5, twistLimitDegrees=5}
 {
  bone1Name=dyn_chest_strap_a_base_l
  bone2Name=spine_03
  swing1LimitDegrees=45
  swing1Motion=Free
  swing2LimitDegrees=45
  swing2Motion=Free
  twistLimitDegrees=45
  twistMotion=Free
 }
 {
  bone1Name=dyn_chest_strap_a_base_r
  bone2Name=spine_03
  swing1LimitDegrees=45
  swing1Motion=Free
  swing2LimitDegrees=45
  swing2Motion=Free
  twistLimitDegrees=45
  twistMotion=Free
 }
 {
  bone1Name=dyn_shldrA_base_l
  bone2Name=spine_03
  swing1LimitDegrees=45
  swing1Motion=Free
  swing2LimitDegrees=45
  swing2Motion=Free
  twistLimitDegrees=45
  twistMotion=Free
 }
 {bone1Name=dyn_shldrA_l, bone2Name=dyn_shldrA_base_l, swing1LimitDegrees=5, swing2LimitDegrees=15, twistLimitDegrees=5}
 {bone1Name=dyn_shldrA_ring_l, bone2Name=dyn_shldrA_l, swing1LimitDegrees=3, swing2LimitDegrees=20, twistLimitDegrees=3}
 {bone1Name=dyn_grenade_l, bone2Name=dyn_grenade_base_l, swing1LimitDegrees=50, swing2LimitDegrees=60, twistLimitDegrees=5}
 {
  bone1Name=spine_03
  bone2Name=pelvis
  swing1LimitDegrees=45
  swing1Motion=Free
  swing2LimitDegrees=45
  swing2Motion=Free
  twistLimitDegrees=45
  twistMotion=Free
 }
 {
  bone1Name=dyn_grenade_base_l
  bone2Name=pelvis
  swing1LimitDegrees=45
  swing1Motion=Free
  swing2LimitDegrees=45
  swing2Motion=Free
  twistLimitDegrees=45
  twistMotion=Free
 }
 {
  bone1Name=thigh_l
  bone2Name=pelvis
  swing1LimitDegrees=45
  swing1Motion=Free
  swing2LimitDegrees=45
  swing2Motion=Free
  twistLimitDegrees=45
  twistMotion=Free
 }
 {
  bone1Name=dyn_pouchC_base_l
  bone2Name=pelvis
  swing1LimitDegrees=45
  swing1Motion=Free
  swing2LimitDegrees=45
  swing2Motion=Free
  twistLimitDegrees=45
  twistMotion=Free
 }
 {bone1Name=dyn_pouchC_l, bone2Name=dyn_pouchC_base_l, swing1LimitDegrees=15, swing2LimitDegrees=30, twistLimitDegrees=10}
 {
  bone1Name=thigh_r
  bone2Name=pelvis
  swing1LimitDegrees=45
  swing1Motion=Free
  swing2LimitDegrees=45
  swing2Motion=Free
  twistLimitDegrees=45
  twistMotion=Free
 }
 {
  bone1Name=dyn_pouchA_base_r
  bone2Name=thigh_r
  swing1LimitDegrees=45
  swing1Motion=Free
  swing2LimitDegrees=45
  swing2Motion=Free
  twistLimitDegrees=45
  twistMotion=Free
 }
 {
  bone1Name=dyn_pouchB_base_r
  bone2Name=thigh_r
  swing1LimitDegrees=45
  swing1Motion=Free
  swing2LimitDegrees=45
  swing2Motion=Free
  twistLimitDegrees=45
  twistMotion=Free
 }
 {bone1Name=dyn_pouchA_r, bone2Name=dyn_pouchA_base_r, swing1LimitDegrees=30, swing2LimitDegrees=30, twistLimitDegrees=5}
 {bone1Name=dyn_pouchB_r, bone2Name=dyn_pouchB_base_r, swing1LimitDegrees=30, swing2LimitDegrees=30, twistLimitDegrees=5}
 {
  bone1Name=dyn_pouchC_base_r
  bone2Name=pelvis
  swing1LimitDegrees=45
  swing1Motion=Free
  swing2LimitDegrees=45
  swing2Motion=Free
  twistLimitDegrees=45
  twistMotion=Free
 }
 {bone1Name=dyn_pouchC_r, bone2Name=dyn_pouchC_base_r, swing1LimitDegrees=30, swing2LimitDegrees=30, twistLimitDegrees=5}
 {
  bone1Name=dyn_strapCol
  bone2Name=thigh_l
  swing1LimitDegrees=45
  swing1Motion=Free
  swing2LimitDegrees=45
  swing2Motion=Free
  twistLimitDegrees=45
  twistMotion=Free
 }
 {
  bone1Name=dyn_pouchB_base_l
  bone2Name=pelvis
  swing1LimitDegrees=45
  swing1Motion=Free
  swing2LimitDegrees=45
  swing2Motion=Free
  twistLimitDegrees=45
  twistMotion=Free
 }
 {bone1Name=dyn_pouchB_l, bone2Name=dyn_pouchB_base_l, swing1LimitDegrees=15, swing2LimitDegrees=30, twistLimitDegrees=15}
]
