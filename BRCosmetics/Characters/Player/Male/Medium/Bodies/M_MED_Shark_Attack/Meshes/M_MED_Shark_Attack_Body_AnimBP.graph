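# Animation Blueprint for the Shark Attack male body cosmetic, adding secondary motion to straps, pouches and a grenade.
# It copies the attached parent's pose, then constrains gear bones to weighted blends of body bones.
# Thigh rotations measured from ref pose each update drive additive corrective offsets on that gear.
# Two RigidBody nodes finish the chain, relative to pelvis and left thigh; the leg-strap simulation switches off when crouching.
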AnimBlueprint /BRCosmetics/Characters/Player/Male/Medium/Bodies/M_MED_Shark_Attack/Meshes/M_MED_Shark_Attack_Body_AnimBP
target_skeleton=&/M_MED_Shark_Attack_Skeleton
graphs [
 AnimGraph {nodes [
   Anim Dynamics - Bone: dyn_shldrA_l(Alpha=1, ExternalForce="0,0,0") {
    angularDampingOverride=1
    angularSpringConstant=1500
    bAngularSpring=true
    bGravityOverrideInSimSpace=true
    bOverrideAngularDamping=true
    gravityScale=0.35
    simulationSpace=RootRelative
    alphaBoolBlend {blendOption=Linear}
    boundBone {boneName=dyn_shldrA_l}
    physicsBodyDefinitions [{
      sphereCollisionRadius=0
      boundBone {boneName=dyn_shldrA_l}
      boxExtents=<10 20 20>
      constraintSetup {angularLimitsMax=<0 15 0>, angularLimitsMin=<0 -15 0>, angularTarget=<0 0 0>}
      localJointOffset=<10 0 0>
     }]
   }
   "Constraint - dyn_strapCol_base (PPP)"!(ComponentPose, Alpha=1, ConstraintWeights_0=0.2564, ConstraintWeights_1:float=0.7278, ConstraintWeights_2:float=0.0158) -> (Pose) {
    alphaBoolBlend {blendOption=Linear}
    boneToModify {boneName=dyn_strapCol_base}
    constraintSetup [{targetBone {boneName=pelvis}}, {targetBone {boneName=thigh_twist_01_l}}, {targetBone {boneName=spine_01}}]
    constraintWeights [1,1,1]
   }
   "Constraint - dyn_base_legStrap_base_l (PPP)"!(ComponentPose, Alpha=1, ConstraintWeights_0=0.1482, ConstraintWeights_1:float=0.0759, ConstraintWeights_2:float=0.7759) -> (Pose) {
    alphaBoolBlend {blendOption=Linear}
    boneToModify {boneName=dyn_base_legStrap_base_l}
    constraintSetup [{targetBone {boneName=spine_01}}, {targetBone {boneName=thigh_twist_01_l}}, {targetBone {boneName=pelvis}}]
    constraintWeights [1,1,1]
   }
   n1 = Get legUp_R() -> (legUp_R:double)
   n2 = "Transform (Modify) Bone - Bone: dyn_pouchA_base_r"(Alpha=@n1, Translation="0,-1,-4.5", Rotation="-50,-10,50", Scale="1,1,1") {
    rotationMode=BMM_Additive
    rotationSpace=BCS_BoneSpace
    scaleMode=BMM_Additive
    scaleSpace=BCS_BoneSpace
    translationMode=BMM_Additive
    translationSpace=BCS_BoneSpace
    alphaBoolBlend {blendOption=Linear}
    alphaScaleBiasClamp {bMapRange=true, inRange {max=0, min=-100}, outRange {max=0, min=1}}
    boneToModify {boneName=dyn_pouchA_base_r}
    rotation=<rot -50 -10 50>
    translation {y=-1, z=-4.5}
   }
   "Transform (Modify) Bone - Bone: dyn_pouchB_base_r"(ComponentPose=@n2, Alpha=@n1, Translation="-3,-3,-4.5", Rotation="-50,-0.000,40.000", Scale="1,1,1") {
    rotationMode=BMM_Additive
    rotationSpace=BCS_BoneSpace
    scaleMode=BMM_Additive
    scaleSpace=BCS_BoneSpace
    translationMode=BMM_Additive
    translationSpace=BCS_BoneSpace
    alphaBoolBlend {blendOption=Linear}
    alphaScaleBiasClamp {bMapRange=true, inRange {max=0, min=-100}, outRange {max=0, min=1}}
    boneToModify {boneName=dyn_pouchB_base_r}
    rotation {pitch=-50, roll=40}
    translation=<-3 -3 -4.5>
   }
   n3 = "Constraint - dyn_shldrA_base_l (PPP)"!(
    ComponentPose=Local To Component(LocalPose=Copy Pose From Mesh() {bUseAttachedParent=true})
    Alpha=1
    ConstraintWeights_0=0.011
    ConstraintWeights_1:float=0.325
    ConstraintWeights_2:float=0.664
   ) -> (Pose) {
    alphaBoolBlend {blendOption=Linear}
    boneToModify {boneName=dyn_shldrA_base_l}
    constraintSetup [{targetBone {boneName=spine_04}}, {targetBone {boneName=spine_05}}, {targetBone {boneName=clavicle_l}}]
    constraintWeights [1,1,1]
   }
   n4 = "Constraint - dyn_chest_strap_a_base_l (PPPPPP)"!(
    ComponentPose=@n3
    Alpha=1
    ConstraintWeights_0=0.1758
    ConstraintWeights_1:float=0.0488
    ConstraintWeights_2:float=0.1477
    ConstraintWeights_3:float=0.3194
    ConstraintWeights_4:float=0.2891
    ConstraintWeights_5:float=0.0192
   ) -> (Pose) {
    alphaBoolBlend {blendOption=Linear}
    boneToModify {boneName=dyn_chest_strap_a_base_l}
    constraintSetup [
     {targetBone {boneName=clavicle_l}}
     {targetBone {boneName=spine_01}}
     {targetBone {boneName=spine_02}}
     {targetBone {boneName=spine_03}}
     {targetBone {boneName=spine_04}}
     {targetBone {boneName=spine_05}}
    ]
    constraintWeights [1,1,1,1,1,1]
   }
   n5 = "Constraint - dyn_chest_strap_a_base_r (PPPPPP)"!(
    ComponentPose=@n4
    Alpha=1
    ConstraintWeights_0=0.0465
    ConstraintWeights_1:float=0.1448
    ConstraintWeights_2:float=0.3063
    ConstraintWeights_3:float=0.2977
    ConstraintWeights_4:float=0.0227
    ConstraintWeights_5:float=0.182
   ) -> (Pose) {
    alphaBoolBlend {blendOption=Linear}
    boneToModify {boneName=dyn_chest_strap_a_base_r}
    constraintSetup [
     {targetBone {boneName=spine_01}}
     {targetBone {boneName=spine_02}}
     {targetBone {boneName=spine_03}}
     {targetBone {boneName=spine_04}}
     {targetBone {boneName=spine_05}}
     {targetBone {boneName=clavicle_r}}
    ]
    constraintWeights [1,1,1,1,1,1]
   }
   n6 = "Constraint - dyn_pouchA_base_l (PPP)"!(ComponentPose=@n5, Alpha=1, ConstraintWeights_0=0.2826, ConstraintWeights_1:float=0.0604, ConstraintWeights_2:float=0.6569) -> (Pose) {
    alphaBoolBlend {blendOption=Linear}
    boneToModify {boneName=dyn_pouchA_base_l}
    constraintSetup [{targetBone {boneName=spine_01}}, {targetBone {boneName=thigh_twist_01_l}}, {targetBone {boneName=pelvis}}]
    constraintWeights [1,1,1]
   }
   n7 = "Constraint - dyn_pouchB_base_l (PPPP)"!(
    ComponentPose=@n6
    Alpha=1
    ConstraintWeights_0=0.388
    ConstraintWeights_1:float=0.0187
    ConstraintWeights_2:float=0.0102
    ConstraintWeights_3:float=0.5832
   ) -> (Pose) {
    alphaBoolBlend {blendOption=Linear}
    boneToModify {boneName=dyn_pouchB_base_l}
    constraintSetup [{targetBone {boneName=spine_01}}, {targetBone {boneName=spine_02}}, {targetBone {boneName=spine_03}}, {targetBone {boneName=pelvis}}]
    constraintWeights [1,1,1,1]
   }
   n8 = "Constraint - dyn_pouchC_base_l (PPP)"!(ComponentPose=@n7, Alpha=1, ConstraintWeights_0=0.1563, ConstraintWeights_1:float=0.0321, ConstraintWeights_2:float=0.8116) -> (Pose) {
    alphaBoolBlend {blendOption=Linear}
    boneToModify {boneName=dyn_pouchC_base_l}
    constraintSetup [{targetBone {boneName=spine_01}}, {targetBone {boneName=thigh_twist_01_l}}, {targetBone {boneName=pelvis}}]
    constraintWeights [1,1,1]
   }
   n9 = "Constraint - dyn_pouchA_base_r (PPP)"!(ComponentPose=@n8, Alpha=1, ConstraintWeights_0=0.0207, ConstraintWeights_1:float=0.7241, ConstraintWeights_2:float=0.2552) -> (Pose) {
    alphaBoolBlend {blendOption=Linear}
    boneToModify {boneName=dyn_pouchA_base_r}
    constraintSetup [{targetBone {boneName=spine_01}}, {targetBone {boneName=thigh_twist_01_r}}, {targetBone {boneName=pelvis}}]
    constraintWeights [1,1,1]
   }
   n10 = "Constraint - dyn_pouchB_base_r (PPP)"!(ComponentPose=@n9, Alpha=1, ConstraintWeights_0=0.0298, ConstraintWeights_1:float=0.6023, ConstraintWeights_2:float=0.3679) -> (Pose) {
    alphaBoolBlend {blendOption=Linear}
    boneToModify {boneName=dyn_pouchB_base_r}
    constraintSetup [{targetBone {boneName=spine_01}}, {targetBone {boneName=thigh_twist_01_r}}, {targetBone {boneName=pelvis}}]
    constraintWeights [1,1,1]
   }
   n11 = "Constraint - dyn_pouchC_base_r (PPPP)"!(
    ComponentPose=@n10
    Alpha=1
    ConstraintWeights_0=0.2721
    ConstraintWeights_1:float=0.0421
    ConstraintWeights_2:float=0.0162
    ConstraintWeights_3:float=0.6696
   ) -> (Pose) {
    alphaBoolBlend {blendOption=Linear}
    boneToModify {boneName=dyn_pouchC_base_r}
    constraintSetup [{targetBone {boneName=spine_01}}, {targetBone {boneName=spine_02}}, {targetBone {boneName=spine_03}}, {targetBone {boneName=pelvis}}]
    constraintWeights [1,1,1,1]
   }
   n12 = "Transform (Modify) Bone - Bone: dyn_pouchC_base_r"(ComponentPose=@n11, Alpha=1, Translation="0,0,0.7", Rotation="5,0,-5", Scale="1,1,1") {
    rotationMode=BMM_Additive
    rotationSpace=BCS_BoneSpace
    scaleMode=BMM_Additive
    scaleSpace=BCS_BoneSpace
    translationMode=BMM_Additive
    translationSpace=BCS_BoneSpace
    alphaBoolBlend {blendOption=Linear}
    alphaScaleBiasClamp {inRange {max=-25, min=-100}, outRange {max=0, min=1}}
    boneToModify {boneName=dyn_pouchC_base_r}
    rotation {pitch=5, roll=-5}
    translation {z=0.7}
   }
   n13 = "Constraint - dyn_grenade_base_l (PP)"!(ComponentPose=@n12, Alpha=1, ConstraintWeights_0=0.1835, ConstraintWeights_1:float=0.8165) -> (Pose) {
    alphaBoolBlend {blendOption=Linear}
    boneToModify {boneName=dyn_grenade_base_l}
    constraintSetup [{targetBone {boneName=spine_01}}, {targetBone {boneName=pelvis}}]
    constraintWeights [1,1]
   }
   n14 = "Constraint - dyn_base_legStrap_base_l (R)"(ComponentPose=@n13, Alpha=1, ConstraintWeights_0=1) {
    alphaBoolBlend {blendOption=Linear}
    boneToModify {boneName=dyn_base_legStrap_base_l}
    constraintSetup [{transformType=Rotation, targetBone {boneName=thigh_l}}]
    constraintWeights [1]
   }
   n15 = "Transform (Modify) Bone - Bone: dyn_base_legStrap_base_l"(ComponentPose=@n14, Alpha=Get legUp_L() -> (legUp_L:double), Translation="-2,2.5,0", Rotation="0,35,30", Scale="1,1,1") {
    rotationMode=BMM_Additive
    rotationSpace=BCS_BoneSpace
    scaleMode=BMM_Additive
    scaleSpace=BCS_BoneSpace
    translationMode=BMM_Additive
    translationSpace=BCS_BoneSpace
    alphaBoolBlend {blendOption=Linear}
    alphaScaleBiasClamp {bMapRange=true, inRange {max=-25, min=-100}, outRange {max=0, min=1}}
    boneToModify {boneName=dyn_base_legStrap_base_l}
    rotation {roll=30, yaw=35}
    translation=<-2 2.5>
   }
   n16 = Get legUp_L()
   n17 = "Transform (Modify) Bone - Bone: dyn_pouchA_base_l"(ComponentPose=@n15, Alpha=@n16, Translation="-1,-3,0", Rotation="20,0,-20", Scale="0.8,0.8,0.8") {
    rotationMode=BMM_Additive
    rotationSpace=BCS_BoneSpace
    scaleMode=BMM_Additive
    scaleSpace=BCS_BoneSpace
    translationMode=BMM_Additive
    translationSpace=BCS_BoneSpace
    alphaBoolBlend {blendOption=Linear}
    alphaScaleBiasClamp {bMapRange=true, inRange {max=0, min=-100}, outRange {max=0, min=1}}
    boneToModify {boneName=dyn_pouchA_base_l}
    rotation {pitch=20, roll=-20}
    scale=<0.8 0.8 0.8>
    translation=<-1 -3>
   }
   n18 = "Transform (Modify) Bone - Bone: dyn_pouchB_base_l"(ComponentPose=@n17, Alpha=@n16, Translation="0,-5,0.3", Rotation="5,0,-30", Scale="0.8,0.8,0.8") {
    rotationMode=BMM_Additive
    rotationSpace=BCS_BoneSpace
    scaleMode=BMM_Additive
    scaleSpace=BCS_BoneSpace
    translationMode=BMM_Additive
    translationSpace=BCS_BoneSpace
    alphaBoolBlend {blendOption=Linear}
    alphaScaleBiasClamp {bMapRange=true, inRange {max=0, min=-100}, outRange {max=0, min=1}}
    boneToModify {boneName=dyn_pouchB_base_l}
    rotation {pitch=5, roll=-30}
    scale=<0.8 0.8 0.8>
    translation {y=-5, z=0.3}
   }
   n19 = "Transform (Modify) Bone - Bone: dyn_pouchC_base_l"(ComponentPose=@n18, Alpha=@n16, Translation="0,-3,0.3", Rotation="-5,-5,-15", Scale="0.8,0.8,0.8") {
    rotationMode=BMM_Additive
    rotationSpace=BCS_BoneSpace
    scaleMode=BMM_Additive
    scaleSpace=BCS_BoneSpace
    translationMode=BMM_Additive
    translationSpace=BCS_BoneSpace
    alphaBoolBlend {blendOption=Linear}
    alphaScaleBiasClamp {bMapRange=true, inRange {max=0, min=-100}, outRange {max=0, min=1}}
    boneToModify {boneName=dyn_pouchC_base_l}
    rotation=<rot -5 -5 -15>
    scale=<0.8 0.8 0.8>
    translation {y=-3, z=0.3}
   }
   n20 = "Transform (Modify) Bone - Bone: dyn_grenade_base_l"(ComponentPose=@n19, Alpha=@n16, Translation="0,0,0", Rotation="-5,-10,0", Scale="1,1,1") {
    rotationMode=BMM_Additive
    rotationSpace=BCS_BoneSpace
    scaleMode=BMM_Additive
    scaleSpace=BCS_BoneSpace
    translationMode=BMM_Additive
    translationSpace=BCS_BoneSpace
    alphaBoolBlend {blendOption=Linear}
    alphaScaleBiasClamp {bMapRange=true, inRange {max=0, min=-100}, outRange {max=0, min=1}}
    boneToModify {boneName=dyn_grenade_base_l}
    rotation {pitch=-5, yaw=-10}
   }
   n21 = Get legUp_R()
   n22 = "Transform (Modify) Bone - Bone: dyn_pouchA_base_r"(ComponentPose=@n20, Alpha=@n21, Translation="2,1.5,-2.5", Rotation="-40,0,40", Scale="1,1,1") {
    rotationMode=BMM_Additive
    rotationSpace=BCS_BoneSpace
    scaleMode=BMM_Additive
    scaleSpace=BCS_BoneSpace
    translationMode=BMM_Additive
    translationSpace=BCS_BoneSpace
    alphaBoolBlend {blendOption=Linear}
    alphaScaleBiasClamp {bMapRange=true, inRange {max=0, min=-100}, outRange {max=0, min=1}}
    boneToModify {boneName=dyn_pouchA_base_r}
    rotation {pitch=-40, roll=40}
    translation=<2 1.5 -2.5>
   }
   n23 = "Transform (Modify) Bone - Bone: dyn_pouchB_base_r"(ComponentPose=@n22, Alpha=@n21, Translation="-2,0.5,-3", Rotation="-40,-0.000,40.000", Scale="1,1,1") {
    rotationMode=BMM_Additive
    rotationSpace=BCS_BoneSpace
    scaleMode=BMM_Additive
    scaleSpace=BCS_BoneSpace
    translationMode=BMM_Additive
    translationSpace=BCS_BoneSpace
    alphaBoolBlend {blendOption=Linear}
    alphaScaleBiasClamp {bMapRange=true, inRange {max=0, min=-100}, outRange {max=0, min=1}}
    boneToModify {boneName=dyn_pouchB_base_r}
    rotation {pitch=-40, roll=40}
    translation=<-2 0.5 -3>
   }
   n24 = "Transform (Modify) Bone - Bone: dyn_pouchA_base_r"(ComponentPose=@n23, Alpha=Get legSide_R() -> (legSide_R:double), Translation="3,0,0", Rotation="0,0,0", Scale="1,1,1") {
    rotationMode=BMM_Additive
    rotationSpace=BCS_BoneSpace
    scaleMode=BMM_Additive
    scaleSpace=BCS_BoneSpace
    translationMode=BMM_Additive
    translationSpace=BCS_BoneSpace
    alphaBoolBlend {blendOption=Linear}
    alphaScaleBiasClamp {bMapRange=true, inRange {max=30}}
    boneToModify {boneName=dyn_pouchA_base_r}
    translation {x=3}
   }
   n25 = "Transform (Modify) Bone - Bone: dyn_strapCol"(ComponentPose=@n24, Alpha=Get legUp_L(), Translation="0,-10,0", Rotation="0,0,0", Scale="1,1,1") {
    rotationMode=BMM_Additive
    rotationSpace=BCS_BoneSpace
    scaleMode=BMM_Additive
    scaleSpace=BCS_BoneSpace
    translationMode=BMM_Additive
    translationSpace=BCS_BoneSpace
    alphaBoolBlend {blendOption=Linear}
    alphaScaleBiasClamp {bMapRange=true, inRange {max=-80, min=-100}, outRange {max=0, min=1}}
    boneToModify {boneName=dyn_strapCol}
    translation {y=-10}
   }
   n26 = RigidBody(ComponentPose=@n25) {
    bClampLinearTranslationLimitToRefPose=true
    bOverrideWorldGravity=true
    overridePhysicsAsset=&/M_MED_Shark_Attack_Physics
    simulationSpace=BaseBoneSpace
    alphaBoolBlend {blendOption=Linear}
    baseBoneRef {boneName=pelvis}
    componentLinearAccScale=<0.45 0.4 0.4>
    componentLinearVelScale=<0.45 0.4 0.4>
    overrideWorldGravity {z=-980}
   }
   n27 = RigidBody(ComponentPose=@n26, Alpha=Get isCrouching() -> (isCrouching:double)) {
    bClampLinearTranslationLimitToRefPose=true
    bOverrideWorldGravity=true
    overridePhysicsAsset=&/M_MED_Shark_Attack_LegStrap_Physics
    simulationSpace=BaseBoneSpace
    alphaBoolBlend {blendOption=Linear}
    alphaScaleBiasClamp {bMapRange=true, outRange {max=0, min=1}}
    baseBoneRef {boneName=thigh_l}
    overrideWorldGravity {z=-980}
   }
   Output Pose(Result=Component To Local(ComponentPose=@n27)) {layerGroup=None}
  ]}
 EventGraph {nodes [
   Event BlueprintUpdateAnimation()
   > n1 = Sequence!(execute) -> (then_0, then_1, then_2:Exec, then_3:Exec)
   n2 = GetOwnerSkeletalMeshComponent()
   n3 = BreakRotator(InRot=BreakTransform(InTransform=GetDeltaTransformFromRefPose(self=@n2, BoneName=thigh_l, BaseName=pelvis)).Rotation)
   > Set legUp_L(legUp_L:double=@n3.Yaw) -> (Output_Get:double)
   > Set legSide_L(legSide_L:double=@n3.Pitch) -> (Output_Get:double)
   n4 = BreakRotator(InRot=BreakTransform(InTransform=GetDeltaTransformFromRefPose(self=@n2, BoneName=thigh_r, BaseName=pelvis)).Rotation)
   Set legUp_R(execute=@n1.then_1, legUp_R:double=@n4.Yaw) -> (Output_Get:double)
   > Set legSide_R(legSide_R:double=@n4.Pitch) -> (Output_Get:double)
   Set isCrouching(execute=@n1.then_3, isCrouching:double="To Float (Boolean)"(InBool=Get bIsCrouching())) -> (Output_Get:double)
   n5 = GetDeltaTransformFromRefPose(self=@n2, BoneName=upperarm_r, BaseName=clavicle_r)
   Set armUp_R(execute=@n1.then_2, armUp_R:double=BreakRotator(InRot=BreakTransform(InTransform=@n5).Rotation).Pitch) -> (Output_Get:double)
   TryGetPawnOwner()
  ]}
]
variables [
 armUp_R {type=number}
 isCrouching {type=number}
 legSide_L {type=number}
 legSide_R {type=number}
 legUp_L {type=number}
 legUp_R {type=number}
]
